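; EQS context that supplies the pawns of all Athena game participants as an actor set for AI environment queries.
; ProvideActorsSet gathers every FortPlayerPawnAthena with GetAllActorsOfClass, then keeps valid pawns whose FortPlayerStateAthena reports IsAnAthenaGameParticipant.
; The querier inputs are unused, so the search covers the whole world with no distance or team filter.
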(fn ProvideActorsSet
  (params (QuerierObject :type Object%20Reference) (QuerierActor :type Actor%20Object%20Reference))
  ;[0 0.1231 0.15 0.5] Return all the athena game participant player pawns.
  (Utilities|FlowControl|Sequence
    (:then_0
      (bind _outactors (Actor|GetAllActorsOfClass "/Script/FortniteGame.FortPlayerPawnAthena"))
      (for _array_element _outactors
        (Utilities|IsValid _array_element
          (:Is%20Valid
            (bind _asfort_player_state_athena (Utilities|Casting|CastToFortPlayerStateAthena (Variables|/Script/Engine.Pawn|GetPlayerState _array_element)))
            (if (Class|FortPlayerStateAthena|IsAnAthenaGameParticipant _asfort_player_state_athena)
              (Utilities|Array|Add (LocalVariables|GetReturnActors) _array_element))))))
    (:then_1
      (return (LocalVariables|GetReturnActors)))))
