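; EQS query context that supplies every FortPawn in the world as reference actors for environment queries.
; ProvideActorsSet ignores the querier and returns the result of GetAllActorsOfClass on the FortPawn class, with no filtering by team or state.
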(fn ProvideActorsSet
  (params (QuerierObject :type Object%20Reference) (QuerierActor :type Actor%20Object%20Reference))
  (bind _outactors (Actor|GetAllActorsOfClass "/Script/FortniteGame.FortPawn"))
  (return _outactors))
