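; EQS context that supplies the center point for an AI pawn's hunting location generator.
; It targets the blackboard GoalActor, falling back to the closest player pawn, or the querier itself, for wandering.
; It clamps half the distance to that target into a hunting radius, then offsets the center from the querier by twice that radius scaled by a signed bias toward or away from the target.
; An optional debug flag draws the points and radius.
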(fn ProvideSingleLocation
  (params (QuerierObject :type Object%20Reference) (QuerierActor :type Actor%20Object%20Reference))
  (bind _querierlocation (LocalVariables|GetQuerierLocation))
  (bind _returnvalue (Fortnite|GetClosestFortPlayerPawn _querierlocation))
  (bind _querier (LocalVariables|GetQuerier))
  ;[1 1 1] Set Querier
  (bind _asfort_aipawn (Utilities|Casting|CastToFortAIPawn QuerierActor)
    (:then
      (LocalVariables|SetQuerier _asfort_aipawn)
      ;[1 1 1] Get Location of two objects.
      (LocalVariables|SetQuerierLocation (Transformation|GetActorLocation _querier))
      (bind _asactor (Utilities|Casting|CastToActor (AI|Components|Blackboard|GetValueasObject (AI|GetBlackboard _querier) (Utilities|Name|MakeLiteralName "GoalActor")))
        (:then
          (LocalVariables|SetGoalActorLocation (Transformation|GetActorLocation _asactor)))
        (:CastFailed
          ;[1 1 1] No GoalActor? You're Wandering! Wander towards a Player, or if no Players exist, just Wander around yourself.
          (Utilities|IsValid _returnvalue
            (:Is%20Valid
              (LocalVariables|SetGoalActorLocation (Transformation|GetActorLocation _returnvalue)) ; No GoalActor, but I can Wander towards a Player.
            )
            (:Is%20Not%20Valid
              (LocalVariables|SetGoalActorLocation (LocalVariables|GetQuerierLocation)) ; No GoalActor and no Players. Just use self as location.
            ))))
      ;[1 1 1] Distance between two objects.
      (LocalVariables|SetDistanceBetweenQuerierAndGoal (Math|Vector|VectorLength (- (LocalVariables|GetQuerierLocation) (LocalVariables|GetGoalActorLocation))))
      ;[1 1 1] Clamp radius between expected min/max.
      (LocalVariables|SetFinalHuntingRadiusToUse (Math|Float|Clamp(Float) (/ (LocalVariables|GetDistanceBetweenQuerierAndGoal) 2) (LocalVariables|GetHuntRadiusMin) (LocalVariables|GetHuntRadiusMax)))
      ;[1 1 1] Make sure center of Hunting doesn't go too far toward the GoalActor if the Radius has been clamped to Max.
      (LocalVariables|SetFinalHuntingRadiusLocationToUse (+ (LocalVariables|GetQuerierLocation) (* (Math|Vector|SelectVector (* (Math|Vector|Normalize (- (LocalVariables|GetGoalActorLocation) (LocalVariables|GetQuerierLocation))) -1) (Math|Vector|Normalize (- (LocalVariables|GetGoalActorLocation) (LocalVariables|GetQuerierLocation))) (< (LocalVariables|GetHuntPercentBiasTowardHunted) 0.0)) (* (* (LocalVariables|GetFinalHuntingRadiusToUse) 2) (Math|Float|Absolute(Float) (LocalVariables|GetHuntPercentBiasTowardHunted))))))
      ; Debug?
      (if (Variables|Default|GetDebug)
        (Rendering|Debug|DrawDebugSphere (Variables|GetQuerierLocation) 50 12 "(R=1.000000,G=0.000000,B=0.000000,A=1.000000)" 10)
        (Rendering|Debug|DrawDebugLine (Variables|GetQuerierLocation) (Variables|GetGoalActorLocation) "(R=0.994000,G=0.000000,B=0.000000,A=1.000000)" 10 4)
        (Rendering|Debug|DrawDebugSphere (Variables|GetGoalActorLocation) 50 12 "(R=0.000000,G=1.000000,B=0.000000,A=1.000000)" 10)
        (Rendering|Debug|DrawDebugSphere (Variables|GetFinalHuntingRadiusLocationToUse) 50 12 "(R=0.000000,G=0.000000,B=1.000000,A=1.000000)" 10)
        (Rendering|Debug|DrawDebugSphere (Variables|GetFinalHuntingRadiusLocationToUse) (Variables|GetFinalHuntingRadiusToUse) 12 "(R=0.000000,G=0.000000,B=1.000000,A=1.000000)" 10)
        (else
          _)))
    (:CastFailed
      _))
  ;[1 1 1] Final Hunting context location sent to EQS.
  (return (LocalVariables|GetFinalHuntingRadiusLocationToUse)))
