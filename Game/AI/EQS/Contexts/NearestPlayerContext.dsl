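; EQS context that supplies the location of the player pawn nearest to the querier as a single point for environment queries.
; ProvideSingleLocation passes the querier actor's location to GetClosestFortPlayerPawn and returns that pawn's location, falling back to the querier's own location when none is valid.
; A branch writing a local variable replaces a Select node to avoid "accessing None" warnings from pure-node evaluation.
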(fn ProvideSingleLocation
  (params (QuerierObject :type Object%20Reference) (QuerierActor :type Actor%20Object%20Reference))
  (bind _returnvalue (Transformation|GetActorLocation QuerierActor))
  (bind _returnvalue_2 (Fortnite|GetClosestFortPlayerPawn _returnvalue))
  ;[1 1 1] Return Location of nearest Player, or Self if no Player exists.
  ;[1 0.5 0 0.25] Sadly, using a Select node from IsValid boolean spams warnings about accessing None because we don't handle only the input we need to handle on pure nodes.
  ; This warning is a mostly useless warning that forces programmers to "fix" perfectly valid blueprints made by non-programmers, who don't understand why they'd need to check whether something is valid or not.
  ; But... no more spam!
  (if (Utilities|IsValid _returnvalue_2 :!class /Script/Engine.KismetSystemLibrary)
    (LocalVariables|SetLocation (Transformation|GetActorLocation _returnvalue_2))
    (else
      (LocalVariables|SetLocation _returnvalue)))
  (return (LocalVariables|GetLocation)))
